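; Vehicle mod for a spiked front bumper that adds speed-scaled ramming damage to whatever the vehicle hits in front.
; On enable it loads hotfix tunables and binds mesh-hit and pawn-hit events; config replication triggers spawn FX.
; Hits above a minimum speed passing a front check apply capped set-by-caller damage plus an immunity effect against repeat hits.
; Pawn hits require a driver and an on-foot non-teammate; non-player hits also take flat damage.
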(event EventBeginPlay
  ;[1 1 1] EVENT -> BeginPlay
  (bind _asfort_athena_vehicle (Utilities|Casting|CastToFortAthenaVehicle (Components|GetOwner)))
  (Variables|Default|SetOwnerVehicle _asfort_athena_vehicle))

(event VehicleMods|EventOnModConfigReplicated
  (params (ModConfigData :type Fort%20Vehicle%20Mod%20Config%20Object%20Reference))
  (bind _outdatas (Class|FortVehicleModConfig|GetModStaticMeshData ModConfigData))
  (bind _meshindexforfrontcheck (Variables|Default|GetMeshIndexForFrontCheck))
  ;[1 1 1] EVENT -> OnConfigReplicated
  (Utilities|IsValid ModConfigData
    (:Is%20Valid
      (if (Utilities|Array|IsValidIndex _outdatas _meshindexforfrontcheck)
        (bind _output_get (Variables|Default|SetBumperMesh (Utilities|Array|Get(acopy) _outdatas _meshindexforfrontcheck)))
        (Niagara|SpawnSystemAttached "/VehicleMods/FX/Systems/SpikedBumper/NS_Spiked_Spawn.NS_Spiked_Spawn" (Class|FortVehicleInterface|GetSkeletalMeshComponent (Variables|Default|GetOwnerVehicle)) (.AttachSocketName (Utilities|Struct|BreakVehicleModStaticMeshAttachmentData _output_get)))))))

(event VehicleMods|EventOnModEnabled
  ;[1 1 1] EVENT -> OnModEnabled
  (CallFunction|SetHotfixValues)
  (bind _ownervehicle (Variables|Default|GetOwnerVehicle)
    (:then
      (Collision|BindEventtoOnComponentHit (Class|FortVehicleInterface|GetSkeletalMeshComponent _ownervehicle) (delegate Custom|VehicleSkeletalMeshHit))
      (bind _asvalet_basic_car_vehicle (Utilities|Casting|CastToValet_BasicCar_Vehicle _ownervehicle))
      (Default|AssignOnPawnHitbyVehicle _asvalet_basic_car_vehicle (delegate Custom|OnPawnHitByVehicle_Event))
      (Audio|SpawnSoundAttached (Variables|Default|GetBumperAppliedSound) (Variables|/Script/Engine.Actor|GetRootComponent _asvalet_basic_car_vehicle)))
    (:else)))

(event Custom|OnPawnHitByVehicle_Event
  (params (FortPawn :type Fort%20Pawn%20Object%20Reference) (ImpactNormal :type Vector) (LaunchVector :type Vector))
  (bind _returnvalue (CallFunction|GetDriverorVehicleAbilityComp))
  (bind _returnvalue_2 (Vehicle|GetPreviousForwardSpeedKmh (Variables|Default|GetOwnerVehicle)))
  (bind _ownervehicle (Variables|Default|GetOwnerVehicle))
  (bind _returnvalue_3 (Vehicle|GetDriver _ownervehicle))
  (bind _returnvalue_4 (Ability|GetFortAbilitySystemComponent FortPawn))
  (bind _maxdamagevsplayers (Variables|Default|GetMaxDamageVsPlayers))
  (bind _speeddamagemultiplierplayer (Variables|Default|GetSpeedDamageMultiplierPlayer))
  (bind _returnvalue_5 (* _returnvalue_2 _speeddamagemultiplierplayer))
  (bind _returnvalue_6 (Math|Float|Min(Float) _returnvalue_5 _maxdamagevsplayers))
  (bind _ownervehicle_2 (Variables|Default|GetOwnerVehicle))
  ;[1 1 1] EVENT -> OnModEnabled
  (if (> (Vehicle|GetPreviousForwardSpeedKmh (Variables|Default|GetOwnerVehicle)) (Variables|Default|GetMinSpeedToApplyDamage))
    (Default|HitActorPassesinFrontCheck FortPawn (Variables|FrontCheck|GetPawn_FrontCheckBehindBumperDistance) true
      (:Hit%20Actor%20In%20Front
        (Utilities|IsValid _returnvalue_3
          (:Is%20Valid
            (Utilities|IsinVehicle FortPawn
              (:false
                (if (not (Team|OnSameTeam FortPawn _returnvalue_3))
                  (bind _returnvalue_7 (Ability|GameplayEffect|AssignTagSetbyCallerMagnitude (GameplayEffects|MakeOutgoingSpec :self _returnvalue :GameplayEffectClass (Variables|Default|GetSpeedMultiDamageGE) :Level 0.0) "(TagName=\"SetByCaller.Magnitude\")" _returnvalue_6)) ; Apply Damage GE based on Speed
                  (GameplayEffects|ApplyGameplayEffectSpecToTarget _returnvalue _returnvalue_7 _returnvalue_4)
                  (GameplayEffects|ApplyGameplayEffectToTarget :self _returnvalue :GameplayEffectClass (Variables|Default|GetImmunityGE) :Target _returnvalue_4 :Level 0.0) ; So we don't damage the same actor multiple times.
                  (Development|PrintString (Utilities|String|Append "Spiked Bumper Bonus Damage:  " (Utilities|String|ToString(Float) _returnvalue_6)) true true "(R=0.081419,G=1.000000,B=0.000000,A=1.000000)" 3.0)))))) ; Don't Deal bonus damage if there is no driver
      ))))

(event Custom|VehicleSkeletalMeshHit
  (params (HitComponent :type Primitive%20Component%20Object%20Reference) (OtherActor :type Actor%20Object%20Reference) (OtherComp :type Primitive%20Component%20Object%20Reference) (NormalImpulse :type Vector) (Hit :type Hit%20Result%20Structure%20%28by%20ref%29))
  (bind (_bblockinghit _binitialoverlap _time _distance _location) (Collision|BreakHitResult Hit))
  (bind _returnvalue (CallFunction|GetDriverorVehicleAbilityComp))
  (bind _returnvalue_2 (Vehicle|GetPreviousForwardSpeedKmh (Variables|Default|GetOwnerVehicle)))
  (bind _returnvalue_3 (Ability|GetFortAbilitySystemComponent OtherActor))
  (bind _maxdamagevsnonplayer (Variables|Default|GetMaxDamageVsNonPlayer))
  (bind _speeddamagemultipliernonplayer (Variables|Default|GetSpeedDamageMultiplierNonPlayer))
  (bind _returnvalue_4 (* _returnvalue_2 _speeddamagemultipliernonplayer))
  (bind _returnvalue_5 (Math|Float|Min(Float) _returnvalue_4 _maxdamagevsnonplayer))
  (bind _ownervehicle (Variables|Default|GetOwnerVehicle))
  ;[1 1 1] EVENT -> OnModEnabled
  (Utilities|Casting|CastToFortPlayerPawn OtherActor
    (:CastFailed
      (Networking|CanExecuteCosmeticEvents
        (:False
          (if (> (Vehicle|GetPreviousForwardSpeedKmh (Variables|Default|GetOwnerVehicle)) (Variables|Default|GetMinSpeedToApplyDamage))
            (Default|HitActorPassesinFrontCheck OtherActor (Variables|FrontCheck|GetVehicle_FrontCheckBehindBumperDistance_Vehicle) false _location
              (:Hit%20Actor%20In%20Front
                (if (not (Variables|Default|GetShouldDamageAllyVehicles))
                  (if (CallFunction|IsObjectFriendlyVehicle OtherActor)
                    _
                    (else
                      _)))
                (Utilities|IsValid _returnvalue_3
                  (:Is%20Valid
                    (GameplayEffects|ApplyGameplayEffectToTarget :self (CallFunction|GetDriverorVehicleAbilityComp) :GameplayEffectClass (Variables|Default|GetDamageGE) :Target _returnvalue_3 :Level 0.0) ; Apply Flat Damage GE
                    (bind _returnvalue_6 (Ability|GameplayEffect|AssignTagSetbyCallerMagnitude (GameplayEffects|MakeOutgoingSpec :self _returnvalue :GameplayEffectClass (Variables|Default|GetSpeedMultiDamageGE) :Level 0.0) "(TagName=\"SetByCaller.Magnitude\")" _returnvalue_5))
                    (GameplayEffects|ApplyGameplayEffectSpecToTarget _returnvalue _returnvalue_6 _returnvalue_3) ; Apply Damage GE based on Speed
                    (GameplayEffects|ApplyGameplayEffectToTarget :self _returnvalue :GameplayEffectClass (Variables|Default|GetImmunityGE) :Target _returnvalue_3 :Level 0.0) ; So we don't damage the same actor multiple times.
                    (Development|PrintString (Utilities|String|Append "Spiked Bumper Bonus Damage:  " (Utilities|String|ToString(Float) _returnvalue_5)) true true "(R=1.000000,G=0.006044,B=0.000000,A=1.000000)" 3.0)))))))) ; Only Run on Server
    )))

(fn DebugFrontCheck
  (params (HitActor :type Actor%20Object%20Reference))
  (Utilities|IsValid HitActor
    (:Is%20Valid
      (bind _asgameplay_tag_asset_interface (Utilities|Casting|CastToGameplayTagAssetInterface HitActor))
      (if (not (GameplayTags|HasMatchingGameplayTag _asgameplay_tag_asset_interface "(TagName=\"Granted.VehicleMod.SpikedBumper.Immune\")"))
        (Rendering|Debug|DrawDebugSphere (Variables|FrontCheck|GetBumperSocketLocationThisFrame) 1e+01 12 "(R=0.978041,G=0.976708,B=1.000000,A=1.000000)" (Variables|FrontCheck|GetFrontTestDebugDrawTimeSeconds)) ; draw bumper loc
        (Rendering|Debug|DrawDebugSphere (Variables|FrontCheck|GetBumperFrontTestLocationThisFrame) 1e+01 12 "(R=1.000000,G=0.000000,B=0.964583,A=1.000000)" (Variables|FrontCheck|GetFrontTestDebugDrawTimeSeconds)) ; draw trace loc
        (Rendering|Debug|DrawDebugLine (Variables|FrontCheck|GetBumperFrontTestLocationThisFrame) (+ (Variables|FrontCheck|GetVectorFromBumperToPawnThisFrame) (Variables|FrontCheck|GetBumperFrontTestLocationThisFrame)) "(R=0.000000,G=0.215686,B=1.000000,A=1.000000)" (Variables|FrontCheck|GetFrontTestDebugDrawTimeSeconds))
        (Rendering|Debug|DrawDebugLine (Variables|FrontCheck|GetBumperFrontTestLocationThisFrame) (+ (Transformation|GetActorForwardVector (Variables|Default|GetOwnerVehicle)) (Variables|FrontCheck|GetBumperFrontTestLocationThisFrame)) "(R=1.000000,G=0.010689,B=0.000000,A=1.000000)" (Variables|FrontCheck|GetFrontTestDebugDrawTimeSeconds))
        (Rendering|Debug|DrawDebugSphere (+ (Transformation|GetActorLocation (Variables|Default|GetOwnerVehicle)) "0.000000,0.000000,300.000000") 1e+02 12 (Math|Color|SelectColor "(R=0.146088,G=1.000000,B=0.000000,A=1.000000)" "(R=0.515625,G=0.000000,B=0.000000,A=1.000000)" (> (Variables|FrontCheck|GetFrontBumperDotResult) 0)) (Variables|FrontCheck|GetFrontTestDebugDrawTimeSeconds)) ; draw a indicator above vehicle telling check result
      ))))

(fn GetDriverOrVehicleAbilityComp
  (return (select (Utilities|IsValid (Vehicle|GetDriver (Variables|Default|GetOwnerVehicle)) :!class /Script/Engine.KismetSystemLibrary) (Ability|GetFortAbilitySystemComponent (Vehicle|GetDriver (Variables|Default|GetOwnerVehicle))) (Ability|GetFortAbilitySystemComponent (Variables|Default|GetOwnerVehicle)))))

(fn IsObjectFriendlyVehicle
  (params (Object :type Object%20Reference))
  (bind _asfort_athena_vehicle (Utilities|Casting|CastToFortAthenaVehicle Object)
    (:then
      (if (Vehicle|HasDriver _asfort_athena_vehicle)
        (if (Team|OnSameTeam (Vehicle|GetDriver _asfort_athena_vehicle) (Vehicle|GetDriver (Variables|Default|GetOwnerVehicle)))
          (return true)
          (else
            _))
        (else
          _))
      (return false))
    (:CastFailed
      _)))

(fn SetHotfixValues
  (Variables|Default|SetSpeedDamageMultiplierNonPlayer (ScalableFloat|GetValueatLevel (Variables|Hotfixes|GetDamageMultiNonPlayerRow)))
  (Variables|Default|SetSpeedDamageMultiplierPlayer (ScalableFloat|GetValueatLevel (Variables|Hotfixes|GetDamageMultiRow)))
  (Variables|Default|SetShouldDamageAllyVehicles (ScalableFloat|GetValueasBool (Variables|Hotfixes|GetShouldDamageAllyVehiclesRow)))
  (Variables|Default|SetMaxDamageVsPlayers (ScalableFloat|GetValueatLevel (Variables|Hotfixes|GetRowMaxDamageVsPlayers)))
  (Variables|Default|SetMaxDamageVsNonPlayer (ScalableFloat|GetValueatLevel (Variables|Hotfixes|GetRowMaxDamageVsNonPlayers)))
  (Variables|Default|SetMinSpeedToApplyDamage (ScalableFloat|GetValueatLevel (Variables|Hotfixes|GetRowMinSpeedToApplyDamage))))
